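# Material function that generates a halftone RGBK dot pattern from a texture, with per-screen dot size and rotation, for the ShakeCrunch guitar's pickaxe effects.
# Each screen rotates the UVs, quantizes them into cells, samples the texture there and sizes a SphereMask dot from it.
# K uses inverted RGB brightness or, via static switch, alpha.
# Outputs tinted RGB dots darkened by K, plus combined and per-screen masks.
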MaterialFunction /SparksCosmetics/Instrument/Guitar/ShakeCrunch/Effects/Pickaxe/MF_HalftoneDots
graph {
 properties {description="Generates a halftone-style RGBK dot pattern, with inputs to control dot size and rotation per each of the 4 channels (R, G, B, and K). Outputs blended color, and several masks."}
 nodes [
  n1 = Add(
   A=Multiply(A=NamedRerouteUsage(), B=VectorParameter() {group=Halftone_DotPattern, parameterName=HalftoneDots_R_Color, defaultValue=<rgba 1 0 0 0>}.RGB)
   B=Multiply(A=NamedRerouteUsage(), B=VectorParameter() {group=Halftone_DotPattern, parameterName=HalftoneDots_G_Color, defaultValue=<rgba 0 1 0 0>}.RGB)
  )
  n2 = Multiply(A=NamedRerouteUsage(), B=VectorParameter() {group=Halftone_DotPattern, parameterName=HalftoneDots_B_Color, defaultValue=<rgba 0 0 1 0>}.RGB)
  n3 = NamedRerouteUsage()
  FunctionOutput(None=Multiply(A=Add(A=@n1, B=@n2), B=OneMinus(None=@n3))) {
   description="Outputs color combined for all 4 screens. Note that the background is by default black, but the Mask outputs can be used to change this."
   outputName=RGBK
  }
  n4 = Add(A=NamedRerouteUsage(), B=NamedRerouteUsage())
  n5 = NamedRerouteUsage()
  FunctionOutput(None=Add(A=Add(A=@n4, B=@n5), B=NamedRerouteUsage())) {description=Outputs a mask combining the modulated outputs of all 4 screens., outputName=Mask_RGBK, sortPriority=10}
  FunctionOutput(None=NamedRerouteUsage()) {description=Outputs a mask for only the R screen., outputName=Mask_R, sortPriority=20}
  FunctionOutput(None=NamedRerouteUsage()) {description=Outputs a mask for only the G screen., outputName=Mask_G, sortPriority=30}
  FunctionOutput(None=NamedRerouteUsage()) {description=Outputs a mask for only the B screen., outputName=Mask_B, sortPriority=40}
  FunctionOutput(None=NamedRerouteUsage()) {description="Outputs a mask for only the K (Black) screen.", outputName=Mask_K, sortPriority=50}
  n6 = FunctionInput(Preview=Constant() {r=80}) {
   bUsePreviewValueAsDefault=true
   description="Float input for UV tile of the dot pattern. A higher value will result in a more dense pattern with smaller dots. It affects all dot screens, same as a printing press."
   inputName=DotScreen_Density
   inputType=FunctionInput_Scalar
   sortPriority=15
  }
  NamedRerouteDeclaration(None=@n6) {name=DotScreen_UVScale, nodeColor=<rgba 0.9453 1 0 1>}
  n7 = FunctionInput(Preview=MaterialFunctionCall(R=Constant() {r=45}, G=Constant(), B=Constant() {r=165}, A=Constant() {r=105}) {materialFunction=&AppendMany}.RGBA) {
   bUsePreviewValueAsDefault=true
   description="Takes a vector 4, where each channel passes a rotation value in degrees to its respective dot screen. "
   inputName=DotScreen_Rotation
   inputType=FunctionInput_Vector4
   sortPriority=30
  }
  NamedRerouteDeclaration(None=Divide(A=@n7) {constB=14.324, desc=Convert to Degrees}) {name=DotScreenRotation_RGBA, nodeColor=<rgba 0 1 0.0859 1>}
  NamedRerouteDeclaration(None=SphereMask(A=Constant() {r=0.5}, B=NamedRerouteUsage(), Radius=NamedRerouteUsage()) {attenuationRadius=0.5}) {name=DotScreen_K, nodeColor=<rgba 0 0 0 1>}
  n8 = NamedRerouteUsage()
  n9 = NamedRerouteUsage()
  n10 = NamedRerouteUsage()
  NamedRerouteDeclaration(None=Multiply(A=TextureSample(UVs=@n8, Tex=@n9).R, B=ComponentMask(None=@n10) {r=true}) {constB=0.5}) {name=DotSize_R, nodeColor=<rgba 1 0.2344 0 1>}
  n11 = NamedRerouteUsage()
  n12 = NamedRerouteUsage()
  n13 = NamedRerouteUsage()
  NamedRerouteDeclaration(None=Multiply(A=TextureSample(UVs=@n11, Tex=@n12).G, B=ComponentMask(None=@n13) {g=true}) {constB=0.5}) {name=DotSize_G, nodeColor=<rgba 1 0.2344 0 1>}
  n14 = NamedRerouteUsage()
  n15 = NamedRerouteUsage()
  n16 = NamedRerouteUsage()
  NamedRerouteDeclaration(None=Multiply(A=TextureSample(UVs=@n14, Tex=@n15).B, B=ComponentMask(None=@n16) {b=true}) {constB=0.5}) {name=DotSize_B, nodeColor=<rgba 1 0.2344 0 1>}
  n17 = TextureSample(UVs=NamedRerouteUsage(), Tex=NamedRerouteUsage())
  n18 = NamedRerouteDeclaration(None=@n17.A) {name=Tex_AlphaChannel, nodeColor=<rgba 1 0.6094 0 1>}
  n19 = ComponentMask(None=NamedRerouteUsage()) {a=true}
  n20 = OneMinus(None=Desaturation(None=@n17.RGB))
  NamedRerouteDeclaration(None=StaticSwitch(True=Multiply(A=@n19, B=@n18) {constB=0.5}, False=Multiply(A=@n20, B=@n19) {constB=0.5}, Value=NamedRerouteUsage())) {name=DotSize_A, nodeColor=<rgba 1 0.2344 0 1>}
  n21 = FunctionInput(Preview=TextureCoordinate()) {
   bUsePreviewValueAsDefault=true
   description=UV space within which to project the Texture and Dot Screens
   inputName=Texture_UVs
   inputType=FunctionInput_Vector2
   sortPriority=5
  }
  NamedRerouteDeclaration(None=@n21) {name=Texture_UVs, nodeColor=<rgba 0 0.1562 1 1>}
  n22 = FunctionInput(Preview=MaterialFunctionCall(R=Constant() {r=0.5}, G=Constant() {r=0.5}, B=Constant() {r=0.5}, A=Constant() {r=0.5}) {materialFunction=&AppendMany}.RGBA) {
   bUsePreviewValueAsDefault=true
   description="Takes a vector 4. Each channel input multiplies against one of 4 dot screen's dot size (defined by the texture input channels). Default is 0.5, which is the largest value retaining the circular dot shape. A value of '1' (if this pixel's texture input is also '1' for this channel) will clip the dot and fill the cel (eventually turning it into a filled square), which may be desirable to get a filled-in look. You can input '0' here to eliminate a channel's dot pattern completely."
   inputName=DotSize_Multiplier
   inputType=FunctionInput_Vector4
   sortPriority=20
  }
  NamedRerouteDeclaration(None=@n22) {name=DotSize_Multiplier_RGBA, nodeColor=<rgba 1 0 0.3984 1>}
  n23 = ComponentMask(None=NamedRerouteUsage()) {r=true}
  n24 = Rotator(Coordinate=NamedRerouteUsage(), Time=@n23)
  n25 = NamedRerouteUsage()
  n26 = Floor(None=Multiply(A=@n24, B=@n25))
  NamedRerouteDeclaration(None=Rotator(Coordinate=Divide(A=@n26, B=@n25), Time=Multiply(A=@n23) {constB=-1})) {name=RotQuantUVs_R, nodeColor=<rgba 1 0 0.0234 1>}
  NamedRerouteDeclaration(None=Frac(None=Multiply(A=@n24, B=@n25))) {name=Frac_DotScreenUVs_R, nodeColor=<rgba 0.4297 1 0 1>}
  n27 = TextureObject() {samplerType=SAMPLERTYPE_LinearColor, texture=/Game/Effects/Fort_Effects/Gameplay/SmokeDustExplosions/Textures/T_NoisePack_LinearColor}
  n28 = FunctionInput(Preview=@n27) {
   bUsePreviewValueAsDefault=true
   description="Texture R, G, and B channels are used to determine each dot screen's dot size (a pixel of '1' in the R texture channel will result in a big output dot for that pixel, and '0' will result in no dot. If the input \"TexAlpha_AsBlack\" bool is 'True', then this Texture input's A channel will be used to generate the Black dot pattern (otherwise black dots are derived from the brightness values of this texture's RGB)."
   inputName=TextureMask
   inputType=FunctionInput_Texture2D
  }
  NamedRerouteDeclaration(None=@n28) {name=TextureObject, nodeColor=<rgba 1 0 0.7969 1>}
  n29 = ComponentMask(None=NamedRerouteUsage()) {g=true}
  n30 = Rotator(Coordinate=NamedRerouteUsage(), Time=@n29)
  n31 = NamedRerouteUsage()
  n32 = Floor(None=Multiply(A=@n30, B=@n31))
  NamedRerouteDeclaration(None=Rotator(Coordinate=Divide(A=@n32, B=@n31), Time=Multiply(A=@n29) {constB=-1})) {name=RotQuantUVs_G, nodeColor=<rgba 1 0 0.0234 1>}
  NamedRerouteDeclaration(None=Frac(None=Multiply(A=@n30, B=@n31))) {name=Frac_DotScreenUVs_G, nodeColor=<rgba 0.4297 1 0 1>}
  n33 = ComponentMask(None=NamedRerouteUsage()) {b=true}
  n34 = Rotator(Coordinate=NamedRerouteUsage(), Time=@n33)
  n35 = NamedRerouteUsage()
  n36 = Floor(None=Multiply(A=@n34, B=@n35))
  NamedRerouteDeclaration(None=Rotator(Coordinate=Divide(A=@n36, B=@n35), Time=Multiply(A=@n33) {constB=-1})) {name=RotQuantUVs_B, nodeColor=<rgba 1 0 0.0234 1>}
  NamedRerouteDeclaration(None=Frac(None=Multiply(A=@n34, B=@n35))) {name=Frac_DotScreenUVs_B, nodeColor=<rgba 0.4297 1 0 1>}
  n37 = ComponentMask(None=NamedRerouteUsage()) {a=true}
  n38 = Rotator(Coordinate=NamedRerouteUsage(), Time=@n37)
  n39 = NamedRerouteUsage()
  n40 = Floor(None=Multiply(A=@n38, B=@n39))
  NamedRerouteDeclaration(None=Rotator(Coordinate=Divide(A=@n40, B=@n39), Time=Multiply(A=@n37) {constB=-1})) {name=RotQuantUVs_K, nodeColor=<rgba 1 0 0.0234 1>}
  NamedRerouteDeclaration(None=Frac(None=Multiply(A=@n38, B=@n39))) {name=Frac_DotScreenUVs_K, nodeColor=<rgba 0.4297 1 0 1>}
  n41 = FunctionInput(Preview=StaticBool()) {
   bUsePreviewValueAsDefault=true
   description="False by default. If this is 'false', then the black dot screen will be derived from the RGB brightness, and the Texture A channel will be ignored. If 'True', then the black dot screen strength will be determined by the A channel of the texture. The A channel mask works the same as the other channels -- white mask pixels applies black dots for this channel, and black produces no dots)."
   inputName=Use TexAlpha as Black?
   inputType=FunctionInput_StaticBool
   sortPriority=50
  }
  NamedRerouteDeclaration(None=@n41) {name=UseTextureAlphaAsBlack?, nodeColor=<rgba 1 0 0.5625 1>}
  NamedRerouteDeclaration(None=SphereMask(A=Constant() {r=0.5}, B=NamedRerouteUsage(), Radius=NamedRerouteUsage()) {attenuationRadius=0.5}) {name=DotScreen_R, nodeColor=<rgba 1 0 0.5391 1>}
  NamedRerouteDeclaration(None=SphereMask(A=Constant() {r=0.5}, B=NamedRerouteUsage(), Radius=NamedRerouteUsage()) {attenuationRadius=0.5}) {name=DotScreen_G, nodeColor=<rgba 0 1 0 1>}
  NamedRerouteDeclaration(None=SphereMask(A=Constant() {r=0.5}, B=NamedRerouteUsage(), Radius=NamedRerouteUsage()) {attenuationRadius=0.5}) {name=DotScreen_B, nodeColor=<rgba 0 0 1 1>}
 ]
}
