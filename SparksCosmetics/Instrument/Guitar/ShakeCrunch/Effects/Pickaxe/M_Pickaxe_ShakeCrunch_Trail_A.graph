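# Translucent, unlit, two-sided Niagara trail material for the ShakeCrunch guitar pickaxe, styled as printed newsprint with halftone dots.
# An emissive adjustment takes a screen-aligned newspaper texture blended into halftone dots along U, darkened by a noise-broken sphere-mask hard edge.
# An opacity chain layers that edge, panning motion lines and leading edges with edge, camera, depth and optional Fresnel fades.
# The graph also carries rainbow, color-curve, HDR and multicolor branches separate from the print look.
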Material /SparksCosmetics/Instrument/Guitar/ShakeCrunch/Effects/Pickaxe/M_Pickaxe_ShakeCrunch_Trail_A
graph {
 properties {
  bUsedWithNiagaraMeshParticles=true
  bUsedWithNiagaraRibbons=true
  bUsedWithNiagaraSprites=true
  blendMode=BLEND_Translucent
  pixelDepthOffsetMode=PDOM_Legacy
  shadingModel=MSM_Unlit
  twoSided=true
  usedShadingModels=Unlit
 }
 nodes [
  n1 = ScalarParameter() {defaultValue=0.03, parameterName=EdgeOffset}
  n2 = ComponentMask(None=TextureCoordinate()) {g=true}
  n3 = ScalarParameter() {defaultValue=0.065, group=Hard Line, parameterName=Edge Radius A, sortPriority=1}
  n4 = ScalarParameter() {defaultValue=0.005, group=Hard Line, parameterName=Edge Radius B, sortPriority=2}
  n5 = ComponentMask(None=TextureCoordinate()) {r=true}
  n6 = ScalarParameter() {defaultValue=0.3, group=Hard Line, parameterName=Edge Hardness A, sortPriority=3}
  n7 = ScalarParameter() {group=Hard Line, parameterName=Edge Hardness B, sortPriority=4}
  n8 = DynamicParameter() {defaultValue=<rgba 0 0 1 1>, paramNames [X Offset, Y Offset, Param3, Param4]}
  n9 = VectorParameter() {group=Soft Leading Edge, parameterName=Soft Leading Edge UV Scale, sortPriority=6, defaultValue=<rgba 0.2 1 0 0>}
  n10 = LinearInterpolate(
   A=TextureCoordinate()
   B=TextureCoordinate() {coordinateIndex=1}
   Alpha=ScalarParameter() {group=UV, parameterName=Use UV1 for Motion Lines A and Leading Edges}
  )
  n11 = ComponentMask(None=@n10) {r=true}
  n12 = LinearInterpolate(A=@n11, B=OneMinus(None=@n11), Alpha=ScalarParameter() {defaultValue=1, group=UV, parameterName=U Direction for Stretch, sortPriority=3})
  n13 = ScalarParameter() {defaultValue=2, group=UV, parameterName=UV Stretch Amount, sortPriority=2}
  n14 = AppendVector(
   A=LinearInterpolate(A=@n11, B=Power(Base=@n12, Exp=@n13), Alpha=ScalarParameter() {group=UV, parameterName=Stretch Trailing Edge, sortPriority=1})
   B=ComponentMask(None=@n10) {g=true}
  )
  n15 = VectorParameter() {group=Soft Leading Edge, parameterName=Soft Leading Edge Panner Speed, sortPriority=7, defaultValue=<rgba -0.65 0 0 0>}
  n16 = Panner(Coordinate=Multiply(A=@n14, B=ComponentMask(None=@n9.RGB) {g=true, r=true}), Speed=ComponentMask(None=@n15.RGB) {g=true, r=true}) {bFractionalPart=true}
  n17 = Add(A=@n16, B=AppendVector(A=@n8.X Offset, B=@n8.X Offset))
  n18 = ScalarParameter() {group=Soft Leading Edge, parameterName=Use_Dynamic_UV-Offsets, sortPriority=0}
  n19 = TextureSampleParameter2D(UVs=Add(A=NamedRerouteUsage(), B=LinearInterpolate(A=@n16, B=@n17, Alpha=@n18))) {
   group=Soft Leading Edge
   parameterName=Soft Leading Edge Texture
   samplerSource=SSM_Wrap_WorldGroupSettings
   sortPriority=4
   texture=&T_NoisePack_01
  }
  n20 = ChannelMaskParameter(None=@n19.RGBA) {group=Soft Leading Edge, parameterName=Soft Leading Edge Texture Mask, sortPriority=5}
  n21 = ScalarParameter() {defaultValue=-0.5, group=Soft Leading Edge, parameterName=Bias, sortPriority=3}
  n22 = TextureCoordinate()
  n23 = Add(
   A=Multiply(A=Add(A=@n20.R, B=@n21), B=ScalarParameter() {defaultValue=0.2, group=Soft Leading Edge, parameterName=Scale, sortPriority=2})
   B=OneMinus(None=ComponentMask(None=@n22) {r=true})
  )
  n24 = Multiply(
   A=SphereMask(A=AppendVector(A=@n1, B=@n2), B=@n1, Radius=LinearInterpolate(A=@n3, B=@n4, Alpha=@n5), Hardness=LinearInterpolate(A=@n6, B=@n7, Alpha=@n5))
   B=@n23
  )
  n25 = Saturate(None=Power(Base=Multiply(A=@n24), Exp=ScalarParameter() {defaultValue=0.1, parameterName=BlackLinePow})) {desc=Hard Edge Mask}
  n26 = ComponentMask(None=TextureCoordinate()) {r=true}
  n27 = TextureCoordinate()
  n28 = ComponentMask(None=@n27) {r=true}
  n29 = LinearInterpolate(
   A=@n28
   B=OneMinus(None=@n28)
   Alpha=Saturate(None=ScalarParameter() {group=Color Curve, parameterName=ColorGradient_HorizontalLeftORRight, sortPriority=2})
  )
  n30 = ComponentMask(None=@n27) {g=true}
  n31 = LinearInterpolate(
   A=@n30
   B=OneMinus(None=@n30)
   Alpha=Saturate(None=ScalarParameter() {group=Color Curve, parameterName=ColorGradient_VerticallUpOrDown, sortPriority=3})
  )
  n32 = ScalarParameter() {group=Color Curve, parameterName=SwizzleGradientDirection, sortPriority=1}
  n33 = CurveAtlasRowParameter(CurveTime=LinearInterpolate(
    A=LinearInterpolate(A=@n29, B=@n31, Alpha=@n32)
    B=@n24
    Alpha=ScalarParameter() {group=Color Curve, parameterName=Hard Line Controls Color Curve, sortPriority=4}
   )) {
   atlas=&AthenaCurveAtlas01
   curve=/Game/Characters/Player/Female/Medium/Bodies/F_MED_HighTower_Squash/Materials/HoloFoilChroma
   defaultValue=237
   group=Color Curve
   parameterName=ColorCurve
  }
  n34 = VectorParameter() {group=Main Texture, parameterName=Main Texture UV Offset, sortPriority=4}
  n35 = VectorParameter() {group=Main Texture, parameterName=Main Texture UV Scale, sortPriority=3, defaultValue=<rgba 1 1 0 0>}
  n36 = Multiply(A=Add(A=@n27, B=ComponentMask(None=@n34.RGB) {g=true, r=true}), B=ComponentMask(None=@n35.RGB) {g=true, r=true})
  n37 = MaterialFunctionCall(XY=@n36) {materialFunction=&Swizzle}
  n38 = ScalarParameter() {group=Main Texture, parameterName=Main Texture Sizzle UVs, sortPriority=5}
  n39 = VectorParameter() {group=Main Texture, parameterName=Main Texture Panner Speed, sortPriority=2}
  n40 = TextureSampleParameter2D(UVs=Panner(Coordinate=LinearInterpolate(A=@n36, B=@n37.YX, Alpha=@n38), Speed=ComponentMask(None=@n39.RGB) {g=true, r=true})) {
   group=Main Texture
   parameterName=Texture
   samplerSource=SSM_Wrap_WorldGroupSettings
   texture=/Game/Effects/Fort_Effects/Textures/Grad/T_HolidayTreeSword_Gradient
  }
  n41 = ChannelMaskParameter(None=@n40.RGBA) {group=Main Texture, maskChannel=Alpha, parameterName=Main Texture Alpha Mask, sortPriority=1, defaultValue=<rgba 0 0 0 1>}
  n42 = Multiply(
   A=LinearInterpolate(B=@n33.A, Alpha=ScalarParameter() {group=Alpha, parameterName=Multiply by Color Curve Alpha}) {constA=1}
   B=LinearInterpolate(B=@n41.R, Alpha=ScalarParameter() {group=Alpha, parameterName=Multiply by Main Texture Alpha Mask, sortPriority=1}) {constA=1}
  )
  n43 = TextureCoordinate()
  n44 = Constant2Vector() {g=0.5, r=1}
  n45 = TextureSample(
   UVs=Panner(Coordinate=Multiply(A=@n43, B=@n44)) {bFractionalPart=true, speedX=0.5}
   Tex=TextureObjectParameter() {group=Color, parameterName=Color Tex, texture=&T_Tile_RainbowColor}
  )
  n46 = Multiply(A=Desaturation(None=@n45.RGB)) {constB=0.5}
  n47 = ParticleColor()
  n48 = VectorParameter() {group=Motion Lines, parameterName=Fine Motion Lines UV Scale A, sortPriority=5, defaultValue=<rgba 1 3.464 0 0>}
  n49 = VectorParameter() {group=Motion Lines, parameterName=Fine Motion Lines Panner Speed A, sortPriority=4, defaultValue=<rgba 0 0.976 0 0>}
  n50 = Panner(Coordinate=Multiply(A=@n14, B=ComponentMask(None=@n48.RGB) {g=true, r=true}), Speed=ComponentMask(None=@n49.RGB) {g=true, r=true})
  n51 = MaterialFunctionCall(XY=@n50) {materialFunction=&Swizzle}
  n52 = ScalarParameter() {defaultValue=1, group=Motion Lines, parameterName=Motion Lines Swizzle UVs, sortPriority=12}
  n53 = NamedRerouteUsage()
  n54 = TextureSampleParameter2D(UVs=Add(A=LinearInterpolate(A=@n50, B=@n51.YX, Alpha=@n52), B=@n53)) {
   group=Motion Lines
   parameterName=Motion Lines Texture
   samplerSource=SSM_Wrap_WorldGroupSettings
   texture=/Game/Effects/Fort_Effects/Textures/Outlander/T_InTheZone_LightRays_01
  }
  n55 = ChannelMaskParameter(None=@n54.RGBA) {group=Motion Lines, parameterName=Motion Lines Texture Mask A, sortPriority=1}
  n56 = ScalarParameter() {defaultValue=1.5, group=Motion Lines, parameterName=Motion Lines Texture A Contrast, sortPriority=2}
  n57 = VectorParameter() {group=Motion Lines, parameterName=Fine Motion Lines UV Scale B, sortPriority=11, defaultValue=<rgba 1 4.4362 0 0>}
  n58 = VectorParameter() {group=Motion Lines, parameterName=Fine Motion Lines Panner Speed B, sortPriority=10, defaultValue=<rgba 0 -0.7616 0 0>}
  n59 = Panner(Coordinate=Multiply(A=@n14, B=ComponentMask(None=@n57.RGB) {g=true, r=true}), Speed=ComponentMask(None=@n58.RGB) {g=true, r=true}) {speedY=-0.7616}
  n60 = MaterialFunctionCall(XY=@n59) {materialFunction=&Swizzle}
  n61 = TextureSampleParameter2D(UVs=Add(A=LinearInterpolate(A=@n59, B=@n60.YX, Alpha=@n52), B=@n53)) {
   group=Motion Lines
   parameterName=Motion Lines Texture
   samplerSource=SSM_Wrap_WorldGroupSettings
   sortPriority=6
   texture=&T_InTheZone_LightRays_01
  }
  n62 = ChannelMaskParameter(None=@n61.RGBA) {group=Motion Lines, maskChannel=Green, parameterName=Motion Lines Texture Mask B, sortPriority=7, defaultValue=<rgba 0 1 0 0>}
  n63 = ScalarParameter() {defaultValue=2, group=Motion Lines, parameterName=Motion Lines Texture B Contrast, sortPriority=8}
  n64 = Add(
   A=Multiply(
    A=ScalarParameter() {defaultValue=1, group=Motion Lines, parameterName=Motion Lines A Intensity, sortPriority=3}
    B=Power(Base=@n55.R, Exp=@n56) {constExponent=1.5}
   )
   B=Multiply(
    A=Power(Base=@n62.R, Exp=@n63)
    B=ScalarParameter() {defaultValue=1, group=Motion Lines, parameterName=Motion Lines B Intensity, sortPriority=9}
   )
  )
  n65 = ScalarParameter() {defaultValue=0.5, group=Motion Lines, parameterName=Combined Motion Lines Scale}
  n66 = Power(Base=@n23, Exp=ScalarParameter() {defaultValue=6, group=Soft Leading Edge, parameterName=Soft Leading Edge Contrast, sortPriority=1}) {constExponent=6}
  n67 = ScalarParameter() {defaultValue=0.1, group=Soft Leading Edge, parameterName=Use Soft Leading Edge}
  n68 = VectorParameter() {group=Hard Leading Edge, parameterName=Hard Leading Edge UV Scale A, sortPriority=4, defaultValue=<rgba 1.2 1.5 0 0>}
  n69 = MaterialFunctionCall(Float2=Multiply(A=@n14, B=ComponentMask(None=@n68.RGB) {g=true, r=true})) {materialFunction=&BreakOutFloat2Components}
  n70 = MaterialFunctionCall(
   X=OneMinus(None=@n69.G)
   Y=Add(A=@n69.R, B=ScalarParameter() {defaultValue=-1, group=Hard Leading Edge, parameterName=Hard Leading Edge Bias A, sortPriority=5}) {constA=0.25, constB=-0.3}
  ) {materialFunction=&MakeFloat2}
  n71 = VectorParameter() {group=Hard Leading Edge, parameterName=Hard Leading Edge UV Scale B, sortPriority=5, defaultValue=<rgba 12 3 0 0>}
  n72 = MaterialFunctionCall(Float2=Multiply(A=@n14, B=ComponentMask(None=@n71.RGB) {g=true, r=true})) {materialFunction=&BreakOutFloat2Components}
  n73 = MaterialFunctionCall(
   X=OneMinus(None=@n72.G)
   Y=Add(A=ScalarParameter() {defaultValue=0.25, group=Hard Leading Edge, parameterName=Hard Leading Edge Bias B, sortPriority=6}, B=@n72.R) {constA=0.25}
  ) {materialFunction=&MakeFloat2}
  n74 = ComponentMask(None=TextureCoordinate()) {g=true}
  n75 = VectorParameter() {group=Hard Leading Edge, parameterName=Hard Edge Mask Panner Speed, sortPriority=7}
  n76 = TextureSampleParameter2D(UVs=Panner(Coordinate=LinearInterpolate(A=@n70.Result, B=@n73.Result, Alpha=@n74), Speed=ComponentMask(None=@n75.RGB) {g=true, r=true})) {
   group=Hard Leading Edge
   parameterName=Hard Leading Edge Texture
   samplerType=SAMPLERTYPE_LinearColor
   sortPriority=2
   texture=/Game/CharClasses/Soldier/Traits/AdvancedTactics/Abilities/GoingCommando/Textures/T_GoingCommando_RingNoise_01
  }
  n77 = OneMinus(None=@n74)
  n78 = Multiply(
   A=ChannelMaskParameter(None=@n76.RGBA) {group=Hard Leading Edge, parameterName=Hard Leading Edge Texture Mask, sortPriority=3}.R
   B=Multiply(A=@n77, B=@n77)
  )
  n79 = ScalarParameter() {defaultValue=1, group=Hard Leading Edge, parameterName=Use Hard Leading Edge}
  n80 = LinearInterpolate(
   A=LinearInterpolate(
    A=ScalarParameter() {defaultValue=0.2, group=Alpha, parameterName=Fine Motion Lines Opacity A, sortPriority=10}
    B=ScalarParameter() {defaultValue=0.3, group=Alpha, parameterName=Fine Motion Lines Opacity B, sortPriority=11}
    Alpha=Multiply(A=@n64, B=@n65)
   ) {constA=1}
   B=ScalarParameter() {defaultValue=1, group=Alpha, parameterName=Leading Edge Opacity, sortPriority=9}
   Alpha=Add(A=Multiply(A=@n66, B=@n67), B=Multiply(A=@n78, B=@n79)) {constB=0}
  ) {constAlpha=0}
  n81 = ComponentMask(None=TextureCoordinate()) {g=true}
  n82 = Multiply(
   A=Add(A=@n81, B=ScalarParameter() {group="Fade Top && Bottom Length Edges", parameterName=Top Length Fade Bias, sortPriority=2})
   B=ScalarParameter() {defaultValue=5, group="Fade Top && Bottom Length Edges", parameterName=Top Length Fade Scale}
  ) {constB=3}
  n83 = Multiply(
   A=Add(A=OneMinus(None=@n81), B=ScalarParameter() {group="Fade Top && Bottom Length Edges", parameterName=Bottom Length Fade Bias, sortPriority=3})
   B=ScalarParameter() {defaultValue=5, group="Fade Top && Bottom Length Edges", parameterName=Bottom Length Fade Scale, sortPriority=1}
  ) {constB=2}
  n84 = Multiply(
   A=LinearInterpolate(B=Saturate(None=@n82), Alpha=ScalarParameter() {defaultValue=1, group=Alpha, parameterName=Fade Top Length Edge, sortPriority=7}) {constA=1}
   B=LinearInterpolate(B=Saturate(None=@n83), Alpha=ScalarParameter() {defaultValue=1, group=Alpha, parameterName=Fade Bottom Length Edge, sortPriority=8}) {constA=1}
  )
  n85 = VectorParameter() {group=Trailing Edge Sphere Mask, parameterName=Trailing Edge Sphereical UV Scale, sortPriority=4, defaultValue=<rgba 0.5 1 0 0>}
  n86 = Multiply(A=TextureCoordinate(), B=ComponentMask(None=@n85.RGB) {g=true, r=true})
  n87 = VectorParameter() {group=Trailing Edge Sphere Mask, parameterName="Trailing Edge Sphere Mask Coord (B)", sortPriority=3, defaultValue=<rgba 0.85 0.82 0 0>}
  n88 = ComponentMask(None=@n87.RGB) {g=true, r=true}
  n89 = ScalarParameter() {defaultValue=0.8, group=Trailing Edge Sphere Mask, parameterName=Trailing Edge Radius, sortPriority=1}
  n90 = ScalarParameter() {defaultValue=0.2, group=Trailing Edge Sphere Mask, parameterName=Trailing Edge Hardness, sortPriority=2}
  n91 = Power(
   Base=OneMinus(None=SphereMask(A=@n86, B=@n88, Radius=@n89, Hardness=@n90))
   Exp=ScalarParameter() {defaultValue=3, group=Trailing Edge Sphere Mask, parameterName=Trailing Edge Contrast}
  ) {constExponent=5}
  n92 = ScalarParameter() {defaultValue=1, group=Alpha, parameterName=Multiply Trailing Edge Sphere Mask, sortPriority=6}
  n93 = Add(
   A=LinearInterpolate(B=@n24, Alpha=ScalarParameter() {defaultValue=5, group=Hard Line, parameterName=Add Top Edge Hard Line})
   B=Multiply(A=Multiply(A=@n80, B=@n84), B=LinearInterpolate(B=@n91, Alpha=@n92) {constA=1})
  )
  n94 = ComponentMask(None=TextureCoordinate()) {r=true}
  n95 = Multiply(
   A=Add(
    A=@n94
    B=ScalarParameter() {defaultValue=0.025, group="Fade Leading && Trailing Edges", parameterName=Leading Edge Fade Bias, sortPriority=2}
   )
   B=ScalarParameter() {defaultValue=5, group="Fade Leading && Trailing Edges", parameterName=Leading Edge Fade Scale}
  ) {constB=3}
  n96 = Multiply(
   A=Add(A=OneMinus(None=@n94), B=ScalarParameter() {group="Fade Leading && Trailing Edges", parameterName=Trailing Edge Fade Bias, sortPriority=3})
   B=ScalarParameter() {defaultValue=3, group="Fade Leading && Trailing Edges", parameterName=Trailing Edge Fade Scale, sortPriority=1}
  ) {constB=2}
  n97 = Multiply(
   A=LinearInterpolate(B=Saturate(None=@n95), Alpha=ScalarParameter() {defaultValue=1, group=Alpha, parameterName=Fade Leading Edge, sortPriority=4}) {constA=1}
   B=LinearInterpolate(B=Saturate(None=@n96), Alpha=ScalarParameter() {defaultValue=1, group=Alpha, parameterName=Fade Trailing Edge, sortPriority=5}) {constA=1}
  )
  n98 = ScalarParameter() {defaultValue=64, group=Alpha, parameterName=NearFadeDistance, sortPriority=3}
  n99 = Multiply(
   A=Multiply(A=Add(A=@n46, B=@n42), B=@n47.A)
   B=Multiply(
    A=Multiply(A=@n93, B=@n97)
    B=MaterialFunctionCall(Distance=@n98) {materialFunction=&Near_Camera_Fade}.Result
   )
  )
  n100 = ShadingPathSwitch(
   Default=DepthFade(Opacity=@n99, FadeDistance=ScalarParameter() {defaultValue=50, group=Alpha, parameterName=FadeDistance, sortPriority=2})
   Mobile=@n99
  ) {desc=Migrated from FeatureLevelSwitch}
  n101 = VertexNormalWS()
  n102 = Multiply(A=MaterialFunctionCall(
    Brightness=ScalarParameter() {defaultValue=1, group=Alpha, parameterName="CameraNormal_Fresnel_Brightness (S)"}
    Exponent=ScalarParameter() {defaultValue=3, group=Alpha, parameterName="CameraNormal_Fresnel_Exponent (S)"}
    Vertex / Pixel Normal=@n101
   ) {materialFunction=&MF_Fresnel_FX}.Fresnel Double Sided)
  Add(
   A=Multiply(A=@n25, B=OneMinus(None=@n26))
   B=LinearInterpolate(A=@n100, B=Multiply(A=@n100, B=@n102), Alpha=ScalarParameter() {group=Alpha, parameterName=Use_CameraNormal_Fresnel})
  )
  n103 = ScalarParameter() {defaultValue=1, group=Multicolor, parameterName=MultiColor_Brightness, sortPriority=1}
  n104 = ScalarParameter() {defaultValue=1, group=Multicolor, parameterName=MultiColor_Exponent, sortPriority=2}
  n105 = TextureObjectParameter() {
   group=Multicolor
   parameterName=MultiColor_GradientTex
   sortPriority=4
   texture=/Game/Characters/Player/Female/Medium/Bodies/F_Med_Soldier_01/Skins/F_MED_COM_Taxi_Upgraded/Skins/Multi/T_multicolor_gradient_v2_D
  }
  n106 = MaterialFunctionCall!(
   Phase Shift=ScalarParameter() {defaultValue=1, group=Multicolor, parameterName=MultiColor_GradientShift, sortPriority=3}
   UV=Clamp(None=MaterialFunctionCall(Brightness=@n103, Exponent=@n104, Vertex / Pixel Normal=@n101) {materialFunction=&MF_Fresnel_FX}.Fresnel Double Sided)
   height texture=Constant()
   Gradient=@n105
   Phases=ScalarParameter() {defaultValue=4, group=Multicolor, parameterName=MultiColor_GradientPhases, sortPriority=5}
  ) {materialFunction=/Game/Characters/Player/Female/Medium/Bodies/F_Med_Soldier_01/Skins/F_MED_COM_Taxi_Upgraded/Skins/Multi/MF_MulticolorGrad}
  LinearInterpolate(
   A=@n106.Result
   B=Add(A=@n106.Result, B=NamedRerouteUsage())
   Alpha=ScalarParameter() {group=Multicolor, parameterName=MultiColor_Add, sortPriority=6}
  )
  n107 = ScalarParameter() {defaultValue=0.005, parameterName=ScreenSpaceTexture_Scale2}
  n108 = Rotator(
   Coordinate=MaterialFunctionCall!(Center Point World Position) {materialFunction=/Game/Packages/Fortnite_UniqueMaterials_1/Effects/WorldAlignedScreenCoordinates}.Result
   Time=Constant() {r=1.5}
  )
  n109 = Multiply(
   A=TextureSample(UVs=Multiply(A=@n107, B=@n108) {constB=5}) {desc=Newpaper projection, texture=&/T_NewspaperTiling}.R
   B=VectorParameter() {defaultValue=<rgba 0.5347 0.4893 0.3193 1>}.RGB
  )
  n110 = ScalarParameter() {defaultValue=0.02, parameterName=ScreenSpaceTexture_Scale}
  n111 = MaterialFunctionCall() {materialFunction=&WorldAlignedScreenCoordinates}
  n112 = ScalarParameter() {defaultValue=0.5, parameterName=Rsize}
  n113 = ScalarParameter() {defaultValue=0.15, parameterName=Gsize}
  n114 = ScalarParameter() {defaultValue=0.5, parameterName=Bsize}
  n115 = ScalarParameter() {parameterName=Ksize}
  n116 = Multiply(
   A=ScalarParameter() {defaultValue=3, parameterName=DotBrightness}
   B=MaterialFunctionCall!(
    TextureMask
    Texture_UVs=Multiply(A=@n110, B=@n111.Result) {constB=5}
    DotScreen_Density=ScalarParameter() {defaultValue=20, parameterName=Dot Density}
    DotSize_Multiplier=MaterialFunctionCall(R=@n112, G=@n113, B=@n114, A=@n115) {materialFunction=&AppendMany}.RGBA
    DotScreen_Rotation=ScalarParameter() {defaultValue=30, parameterName=DotScreen_Rotation}
    Use TexAlpha as Black?
   ) {materialFunction=&/MF_HalftoneDots}.RGBK
  )
  n117 = ScalarParameter() {parameterName=Min}
  n118 = ScalarParameter() {defaultValue=1.25, parameterName=Max}
  n119 = ComponentMask(None=TextureCoordinate()) {r=true}
  n120 = Power(Base=Saturate(None=LinearInterpolate(A=@n117, B=@n118, Alpha=@n119)), Exp=ScalarParameter() {defaultValue=1.25, parameterName=TransitionPow})
  MaterialFunctionCall(Emissive=LinearInterpolate(A=LinearInterpolate(A=@n109, B=@n116, Alpha=@n120), B=ScalarParameter() {defaultValue=0.02, parameterName=Param_2}, Alpha=@n25) {constB=0, desc=Hard Edge}) {materialFunction=&MF_FX_EmissiveAdjustments}
  n121 = Constant3Vector() {constant=<rgba 3 3 3 0>}
  n122 = LinearInterpolate(B=@n33, Alpha=ScalarParameter() {group=Emissive Color, parameterName=Multiply by Color Curve})
  n123 = LinearInterpolate(
   B=LinearInterpolate(
    A=VectorParameter() {group=Two Color Blend, parameterName=Two Color Blend A}.RGB
    B=VectorParameter() {group=Two Color Blend, parameterName=Two Color Blend B, sortPriority=1}.RGB
    Alpha=Add(A=@n93, B=@n122)
   )
   Alpha=ScalarParameter() {group=Emissive Color, parameterName="Multiply by Two Color Blend ", sortPriority=2}
  ) {constA=1}
  n124 = LinearInterpolate(B=@n33, Alpha=ScalarParameter() {group=Emissive Color, parameterName=Multiply by Color Curve}) {constA=1}
  n125 = ScalarParameter() {group=Emissive Color, parameterName="Multiply by Main TextureMask ", sortPriority=1}
  n126 = Multiply(A=Desaturation(None=@n45.RGB, Fraction=ScalarParameter() {defaultValue=-0.85, group=Color, parameterName=Color Tex Desaturation})) {constB=0.75}
  n127 = Power(Base=OneMinus(None=TextureCoordinate()), Exp=Constant2Vector() {g=2, r=6}) {constExponent=4}
  n128 = Constant2Vector() {g=2, r=2}
  n129 = MaterialFunctionCall(Float2=Saturate(None=Multiply(A=@n127, B=@n128) {constB=2})) {materialFunction=&BreakOutFloat2Components}
  n130 = Multiply(
   A=Multiply(A=Multiply(A=@n123, B=@n124), B=LinearInterpolate(B=@n40.RGB, Alpha=@n125) {constA=1})
   B=LinearInterpolate(A=Multiply(A=@n126, B=@n47.RGB), B=LinearInterpolate(A=@n126, B=@n47.RGB, Alpha=@n129.R), Alpha=@n129.R)
  )
  n131 = Clamp(None=@n130)
  n132 = ScalarParameter() {defaultValue=100, group=Derive HDR from LDR, parameterName=HDR Dynamic Range, sortPriority=3}
  n133 = LinearInterpolate(
   B=LinearInterpolate(Alpha=@n24) {constA=0.01, constB=0.8}
   Alpha=ScalarParameter() {group=Derive HDR from LDR, parameterName=Mutilpy HDR Intensity by Hard Line, sortPriority=2}
  ) {constA=1}
  n134 = Multiply(A=ScalarParameter() {defaultValue=1, group=Derive HDR from LDR, parameterName=HDR Intensity, sortPriority=1}, B=@n133)
  n135 = ScalarParameter() {defaultValue=1, group=Derive HDR from LDR, parameterName=Mobile HDR Intensity, sortPriority=4}
  n136 = Multiply(A=@n130, B=@n133)
  n137 = LinearInterpolate(
   A=@n130
   B=ShadingPathSwitch(
    Default=MaterialFunctionCall(LDR Input=@n131, HDR Dynamic Range=@n132, HDR Tint=@n130, HDR Intensity Multiplier=@n134) {materialFunction=&DeriveHDRfromLDR}.RemappedRange
    Mobile=Multiply(A=@n135, B=@n136)
   ) {desc=Migrated from FeatureLevelSwitch}
   Alpha=ScalarParameter() {group=Derive HDR from LDR, parameterName=Use HDR Emissive Color}
  )
  Multiply(
   A=MaterialFunctionCall(Color_Morning=@n121, Color_Day=@n121, Color_Evening=@n121, Color_Night=Constant3Vector() {constant=<rgba 0.8 0.8 2 0>}) {materialFunction=&Color_Time_of_Day}.Result
   B=@n137
  )
  Multiply(A=@n129.R)
  n138 = Multiply(A=VertexColor().R, B=ScalarParameter() {defaultValue=0.5, group=Vertex Color, parameterName=Vertex Color Offset, sortPriority=1})
  n139 = Constant()
  NamedRerouteDeclaration(None=StaticSwitchParameter(True=AppendVector(A=@n138, B=@n139), False=Constant()) {group=Vertex Color, parameterName=VertexColorOffset, sortPriority=0}) {name=VertexColorOffset, nodeColor=<rgba 0 1 0.8125 1>}
  NamedRerouteDeclaration(None=@n137) {name=Emissive, nodeColor=<rgba 0.3828 0 1 1>}
 ]
}
